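# Generic Niagara effect for electrical destruction of world objects, going by its name and folder.
# It holds two emitters, Electric_Arcs and Sparks, under a system that uses only the SystemState module.
# No user parameters, data interfaces, dynamic inputs or renderers are listed, so the emitters' behavior and look are not visible here.
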NiagaraSystem /Game/Effects/Fort_Effects/World/Destruction/Electric/NSLW_Destruction_Electric_Generic
dependencies {usedDataInterfaces [], usedDynamicInputs [], usedModules [&SystemState]}
emitters {Electric_Arcs {rendererClasses [], renderers []}, Sparks {rendererClasses [], renderers []}}
summary {
 systemName=NSLW_Destruction_Electric_Generic
 emitters [{emitterName=Electric_Arcs, rendererClasses []}, {emitterName=Sparks, rendererClasses []}]
 userVariables []
}
