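# Niagara system for a generic electrical destruction effect with smoke, filed under world destruction effects.
# It combines five emitters: two flash layers (one with low spikes), smoke, sparks and electric arcs.
# The flash and smoke emitters carry an LW suffix.
# System State is the only module listed at system level, and no user parameters are exposed; emitter internals and renderers are not shown.
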NiagaraSystem /Game/Effects/Fort_Effects/World/Destruction/Electric/NS_Destruction_Electric_Generic_WithSmoke
dependencies {usedDataInterfaces [], usedDynamicInputs [], usedModules [&SystemState]}
emitters {
 Flash_LW {rendererClasses [], renderers []}
 Flash_Low_Spikes_LW {rendererClasses [], renderers []}
 Smoke_LW {rendererClasses [], renderers []}
 Sparks {rendererClasses [], renderers []}
 electric_arcs {rendererClasses [], renderers []}
}
summary {
 systemName=NS_Destruction_Electric_Generic_WithSmoke
 emitters [
  {emitterName=Flash_LW, rendererClasses []}
  {emitterName=Smoke_LW, rendererClasses []}
  {emitterName=Flash_Low_Spikes_LW, rendererClasses []}
  {emitterName=electric_arcs, rendererClasses []}
  {emitterName=Sparks, rendererClasses []}
 ]
 userVariables []
}
